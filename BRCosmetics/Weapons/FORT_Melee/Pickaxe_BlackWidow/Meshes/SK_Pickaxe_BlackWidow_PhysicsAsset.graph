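# Physics asset for the Black Widow two-handed pickaxe skeletal mesh, a melee weapon cosmetic.
# It holds a single capsule body on the Melee_TwoHanded_Handle bone, giving the pickaxe one rough collision shape.
# There are no constraints, so the asset defines no jointed or ragdoll motion between parts.
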PhysicsAsset /BRCosmetics/Weapons/FORT_Melee/Pickaxe_BlackWidow/Meshes/SK_Pickaxe_BlackWidow_PhysicsAsset
bodies {Melee_TwoHanded_Handle [{shapeType=Capsule, length=68.3594, radius=24.6576, center=<0.6307 2.1992 14.3154>, rotation=<rot -0.3039 -90 -0.0008>}]}
constraints []
